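# Looping ambient SoundCue for the shoreline at the Figment S05 Cube event destination, in the outdoor ambient sound class.
# A quality-level node picks between two looping waves, using the standard shoreline recording for the first two levels and the "_opt" variant for the last two.
# It references the river water attenuation asset but enables override attenuation, so the cue's own settings apply, including a widened stereo spread and a non-spatialized radius.
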SoundCue /Figment_S05_Cube/Athena/Sounds/Events/Cube/Destination/Cube_Figment_Dest_ShorelineLoop_A
root=QualityLevel_0
nodes [
 {id=QualityLevel_0, children [WavePlayer_0,WavePlayer_0,WavePlayer_1,WavePlayer_1]}
 {id=WavePlayer_0, properties {bLooping=true, soundWaveAssetPtr=/FNGameplayCues/Athena/Sounds/Events/Cube/Destination/Cube_Shoreline_01}}
 {id=WavePlayer_1, properties {bLooping=true, soundWaveAssetPtr=/FNGameplayCues/Athena/Sounds/Events/Cube/Destination/Cube_Shoreline_01_opt}}
]
properties {
 attenuationSettings=/Game/Athena/Sounds/Ambient/Fort_Amb_Water_River_Att
 bOverrideAttenuation=true
 maxDistance=3600
 soundClassObject=&Fort_Ambient_Outdoors_Athena
 volumeMultiplier=1
 attenuationOverrides {nonSpatializedRadiusEnd=1000, stereoSpread=2500, attenuationShapeExtents {x=0}}
}
