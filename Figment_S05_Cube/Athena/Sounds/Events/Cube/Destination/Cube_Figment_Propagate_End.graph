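# Sound cue for the end of the Figment Cube's propagate event at its destination, judging by its name and path.
# A single wave player plays Cube_PropagateEnd_01 at reduced volume through an important in-game mission sound class.
# An attenuation override turns off spatialization and sets a very long falloff distance, so the sound carries far without directional panning.
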SoundCue /Figment_S05_Cube/Athena/Sounds/Events/Cube/Destination/Cube_Figment_Propagate_End
root=WavePlayer_1
nodes [{id=WavePlayer_1, properties {soundWaveAssetPtr=&/Cube_PropagateEnd_01}}]
properties {
 bOverrideAttenuation=true
 maxDistance=60000
 soundClassObject=&Fort_Mission_Ingame_Important_PSM
 volumeMultiplier=0.5
 attenuationOverrides {bSpatialize=false, falloffDistance=60000, attenuationShapeExtents {x=0}}
}
