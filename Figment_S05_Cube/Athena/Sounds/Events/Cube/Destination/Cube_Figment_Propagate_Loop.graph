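# Sound cue for the Figment Cube propagation loop, mixing a wide-range bed with two close looping layers.
# The wide branch pairs a delayed, non-looping begin sound with two loops crossfaded by the Corruption parameter.
# Each close layer has its own attenuation, one adding distance low-pass filtering.
# InvCorruption drives continuous volume or pitch modulation across the loops, and quality-level nodes switch between full and optimized waves.
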SoundCue /Figment_S05_Cube/Athena/Sounds/Events/Cube/Destination/Cube_Figment_Propagate_Loop
root=Mixer_0
nodes [
 {
  id=Attenuation_1
  children [QualityLevel_1]
  properties {attenuationSettings=/Figment_S05_Cube/Athena/Sounds/Events/Cube/Cube_Ambient_Close_Att}
 }
 {
  id=Attenuation_2
  children [QualityLevel_2]
  properties {
   bOverrideAttenuation=true
   attenuationOverrides {
    bAttenuateWithLPF=true
    distanceAlgorithm=Custom
    falloffDistance=10000
    lPFFrequencyAtMax=3000
    lPFRadiusMax=5000
    lPFRadiusMin=0
    stereoSpread=2000
    attenuationShapeExtents {x=0}
    customAttenuationCurve {editorCurveData {keys [<at -0.001 0>,<at 0.34 1>,<at 1 0>]}}
   }
  }
 }
 {
  id=Attenuation_3
  children [Mixer_1]
  properties {
   bOverrideAttenuation=true
   attenuationOverrides {
    distanceAlgorithm=Custom
    falloffDistance=30000
    stereoSpread=2996
    attenuationShapeExtents {x=0}
    customAttenuationCurve {editorCurveData {keys [
       {arriveTangent=-3.5624, interpMode=RCIM_Cubic, leaveTangent=-3.5624, tangentMode=RCTM_User, time=0, value=1}
       {arriveTangent=0.1391, interpMode=RCIM_Cubic, leaveTangent=0.1391, tangentMode=RCTM_User, time=1, value=0}
      ]}}
   }
  }
 }
 {id=Delay_1, children [WavePlayer_8], properties {delayMax=3, delayMin=3}}
 {id=Mixer_0, children [Attenuation_3, ModulatorContinuous_1, ModulatorContinuous_2], properties {inputVolume [0.944,1,1]}}
 {id=Mixer_1, children [Delay_1, ModulatorContinuous_4], properties {inputVolume [2,1]}}
 {
  id=ModulatorContinuous_1
  children [Attenuation_1]
  properties {pitchModulationParams {minOutput=0.5, parameterName=InvCorruption}, volumeModulationParams {parameterName=InvCorruption}}
 }
 {
  id=ModulatorContinuous_2
  children [Attenuation_2]
  properties {pitchModulationParams {minOutput=0.5}, volumeModulationParams {parameterName=InvCorruption}}
 }
 {
  id=ModulatorContinuous_4
  children [QualityLevel_0]
  properties {pitchModulationParams {maxOutput=1.2, minOutput=0.8, parameterName=InvCorruption}}
 }
 {
  id=ParamCrossFade_0
  children [WavePlayer_0,WavePlayer_1]
  properties {
   paramName=Corruption
   crossFadeInput [
    {fadeInDistanceEnd=0, fadeInDistanceStart=0, fadeOutDistanceEnd=1, fadeOutDistanceStart=0, volume=1}
    {fadeInDistanceEnd=1, fadeInDistanceStart=0, fadeOutDistanceEnd=100, fadeOutDistanceStart=100, volume=1}
   ]
  }
 }
 {
  id=ParamCrossFade_1
  children [WavePlayer_2,WavePlayer_3]
  properties {
   paramName=Corruption
   crossFadeInput [
    {fadeInDistanceEnd=0, fadeInDistanceStart=0, fadeOutDistanceEnd=1, fadeOutDistanceStart=0, volume=1}
    {fadeInDistanceEnd=1, fadeInDistanceStart=0, fadeOutDistanceEnd=100, fadeOutDistanceStart=100, volume=1}
   ]
  }
 }
 {id=QualityLevel_0, children [ParamCrossFade_0, ParamCrossFade_0, ParamCrossFade_1, ParamCrossFade_1]}
 {id=QualityLevel_1, children [WavePlayer_5,WavePlayer_5,WavePlayer_4,WavePlayer_4]}
 {id=QualityLevel_2, children [WavePlayer_7,WavePlayer_7,WavePlayer_6,WavePlayer_6]}
 {
  id=WavePlayer_0
  properties {bLooping=true, soundWaveAssetPtr=&/Cube_PropagateLayer_01}
 }
 {
  id=WavePlayer_1
  properties {bLooping=true, soundWaveAssetPtr=&/Cube_PropagateLayer_02}
 }
 {
  id=WavePlayer_2
  properties {bLooping=true, soundWaveAssetPtr=&/Cube_PropagateLayer_01_opt}
 }
 {
  id=WavePlayer_3
  properties {bLooping=true, soundWaveAssetPtr=&/Cube_PropagateLayer_02_opt}
 }
 {
  id=WavePlayer_4
  properties {bLooping=true, soundWaveAssetPtr=&/Cube_PropagateClose_01_opt}
 }
 {
  id=WavePlayer_5
  properties {bLooping=true, soundWaveAssetPtr=&/Cube_PropagateClose_01}
 }
 {
  id=WavePlayer_6
  properties {bLooping=true, soundWaveAssetPtr=&Cube_PropagateClose_01_opt}
 }
 {
  id=WavePlayer_7
  properties {bLooping=true, soundWaveAssetPtr=&Cube_PropagateClose_01}
 }
 {id=WavePlayer_8, properties {soundWaveAssetPtr=&/Cube_PropagateBegin_01}}
]
properties {maxDistance=30000, soundClassObject=&Fort_Mission_Ingame_Important_PSM, volumeMultiplier=1}
